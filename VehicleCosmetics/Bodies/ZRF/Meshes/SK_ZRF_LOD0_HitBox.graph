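# Physics asset that, going by its name, serves as the hitbox for the ZRF vehicle body skeletal mesh at LOD0.
# It holds a single body on the root bone, made of one box shape offset forward along X and raised along Z from the bone origin.
# With no constraints or other bodies, the collision is one coarse volume with no per-part or articulated physics.
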PhysicsAsset /VehicleCosmetics/Bodies/ZRF/Meshes/SK_ZRF_LOD0_HitBox
bodies {root [{shapeType=Box, extentX=385.6236, extentY=277.4381, extentZ=123.709, shapeName=root_box, center=<44.4021 0 116.8563>}]}
constraints []
